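# Conversation database for the Bandolette NPC, defining the greeting a player gets and the shared dialogue branches that follow.
# The entry tag leads to one speech node with a default line plus contextual lines for bar count, first-ever conversation and two specific outfits.
# Each contextual line takes its priority from a Tandem.Speech.Priority curve row.
# The speech then branches through link nodes into the shared NPCLibrary Quests and BasicServices conversations by entry tag.
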ConversationDatabase /NPCLibrary/NPCs/Bandolette/Bandolette/NPCLibrary_Conversation_Bandolette
entry_points [{tag=Conversation.Entry.Bandolette, destinations [#1]}]
nodes [
 ConversationEntryPointNode_7 {
  id=#1
  class=ConversationEntryPointNode
  outputs [#2]
  properties {entryTag {tagName=Conversation.Entry.Bandolette}}
 }
 Speech_C_4 {
  id=#2
  class=Speech_C
  outputs [#3, #4]
  properties {
   nodeName=""
   generalConfig {
    defaultMessage=Staying sharp out there?
    contextualMessages [
     {
      message="Wow. I heard your pockets jingling from a mile away. <b>Whatcha saving up for?</>"
      contextRequirements [{requirement {class=FortControllerRequirement_HasBars, properties {thresholdTestConfig {requirement=GreaterThanOrEqual}}}}]
      priority {curve {rowName=Tandem.Speech.Priority.HighTotalBars}}
     }
     {
      message=I think my brother told me about you.
      contextRequirements [{requirement {class=FortControllerRequirement_IsFirstEverConversationWithNPC}}]
      priority {curve {rowName=Tandem.Speech.Priority.FirstEverInteraction}}
     }
     {
      message="Oh, <b>hey bro</>.Didn't expect to see you here."
      contextRequirements [{requirement {
         class=FortControllerRequirement_HasCID
         properties {allowedCIds [/BRCosmetics/Athena/Items/Cosmetics/Characters/CID_131_Athena_Commando_M_Warpaint]}
        }}]
      priority {curve {rowName=Tandem.Speech.Priority.OutfitReaction}}
     }
     {
      message=Hm. Look who it is.
      contextRequirements [{requirement {
         class=FortControllerRequirement_HasCID
         properties {allowedCIds [/BRCosmetics/Athena/Items/Cosmetics/Characters/CID_A_008_Athena_Commando_F_CombatDoll]}
        }}]
      priority {curve {rowName=Tandem.Speech.Priority.OutfitReaction}}
     }
    ]
   }
  }
 }
 ConversationLinkNode_0 {id=#4, class=ConversationLinkNode, properties {remoteEntryTag {tagName=Conversation.Entry.BasicServices}}}
 ConversationLinkNode_1 {id=#3, class=ConversationLinkNode, properties {remoteEntryTag {tagName=Conversation.Entry.NPCLibrary.Quests}}}
]
